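# Animation Blueprint for the BrawnGraze pickaxe that adds secondary physics motion to its dangling eye and leg bones.
# The cached input pose feeds three chains of spring-driven Anim Dynamics nodes, each covering both eyes and the left and right upper and lower legs.
# In the front end, bIsOnPawn picks between two chains using root-relative eyes and world-space legs; otherwise a looser-limited, bone-relative chain runs.
# The event graph has no logic.
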AnimBlueprint /BRCosmetics/Weapons/FORT_Melee/Pickaxe_BrawnGraze/Meshes/BrawnGraze_Axe_AnimBP
target_skeleton=&/BrawnGraze_Axe_Skeleton
graphs [
 AnimGraph {nodes [
   n1 = Anim Dynamics - Bone: dyn_eye_r(ComponentPose=Local To Component(LocalPose=Use cached pose 'InputCachePose'()), Alpha=0.8, ExternalForce="0,0,0") {
    angularDampingOverride=1
    angularSpringConstant=19000
    bAngularSpring=true
    bOverrideAngularDamping=true
    bUseGravityOverride=true
    gravityScale=0.35
    simulationSpace=RootRelative
    worldSpaceMinimumScale=0
    boundBone {boneName=dyn_eye_r}
    componentAppliedLinearAccClamp=<4500 4500 4500>
    componentLinearAccScale=<0.5 0.5 0.5>
    componentLinearVelScale=<0.5 0.5 0.5>
    gravityOverride {z=-980}
    physicsBodyDefinitions [{
      boundBone {boneName=dyn_eye_r}
      boxExtents=<20 19 19>
      constraintSetup {angularLimitsMax=<0 37 37>, angularLimitsMin=<0 -37 -37>, angularTarget=<1 0 0>}
      localJointOffset=<2.9 0 0>
     }]
    simSpaceSettings {maxAngularAcceleration=1000, maxAngularVelocity=100}
   }
   n2 = Anim Dynamics - Bone: dyn_eye_l(ComponentPose=@n1, Alpha=0.8, ExternalForce="0,0,0") {
    angularDampingOverride=1
    angularSpringConstant=19000
    bAngularSpring=true
    bOverrideAngularDamping=true
    bUseGravityOverride=true
    gravityScale=0.35
    simulationSpace=RootRelative
    worldSpaceMinimumScale=0
    boundBone {boneName=dyn_eye_l}
    componentAppliedLinearAccClamp=<4500 4500 4500>
    componentLinearAccScale=<0.5 0.5 0.5>
    componentLinearVelScale=<0.5 0.5 0.5>
    gravityOverride {z=-980}
    physicsBodyDefinitions [{
      boundBone {boneName=dyn_eye_l}
      boxExtents=<20 19 19>
      constraintSetup {angularLimitsMax=<0 32 23>, angularLimitsMin=<0 -23 -32>, angularTarget=<1 0 0>}
      localJointOffset=<2.9 0 0>
     }]
    simSpaceSettings {maxAngularAcceleration=1000, maxAngularVelocity=100}
   }
   n3 = Anim Dynamics - Bone: dyn_upperLeg_r(ComponentPose=@n2, Alpha=1, ExternalForce="0,0,0") {
    angularDampingOverride=1
    angularSpringConstant=15000
    bAngularSpring=true
    bOverrideAngularDamping=true
    bUseGravityOverride=true
    gravityScale=0.35
    simulationSpace=World
    boundBone {boneName=dyn_upperLeg_r}
    componentAppliedLinearAccClamp=<2000 2000 2000>
    componentLinearAccScale=<0.5 0.5 0.5>
    componentLinearVelScale=<0.5 0.5 0.5>
    gravityOverride {z=-980}
    physicsBodyDefinitions [{
      boundBone {boneName=dyn_upperLeg_r}
      boxExtents=<34.5 34.5 34.5>
      constraintSetup {angularLimitsMax=<0 0.5 26>, angularLimitsMin=<0 -0.5 -19>, angularTarget=<1 0 0>}
      localJointOffset=<0.72 0 0>
     }]
   }
   n4 = Anim Dynamics - Bone: dyn_upperLeg_l(ComponentPose=@n3, Alpha=1, ExternalForce="0,0,0") {
    angularDampingOverride=1
    angularSpringConstant=17000
    bAngularSpring=true
    bOverrideAngularDamping=true
    bUseGravityOverride=true
    gravityScale=0.35
    simulationSpace=World
    boundBone {boneName=dyn_upperLeg_l}
    componentAppliedLinearAccClamp=<5000 5000 5000>
    componentLinearAccScale=<0.8 0.8 0.8>
    componentLinearVelScale=<0.8 0.8 0.8>
    gravityOverride {z=-950}
    physicsBodyDefinitions [{
      boundBone {boneName=dyn_upperLeg_l}
      boxExtents=<32.5 32.5 32.5>
      constraintSetup {angularLimitsMax=<0 2 23>, angularLimitsMin=<0 -1 -20>, angularTarget=<1 0 0>}
      localJointOffset=<0.78 0 0>
     }]
   }
   n5 = Anim Dynamics - Bone: dyn_lowerLeg_r(ComponentPose=@n4, Alpha=1, ExternalForce="0,0,0") {
    angularDampingOverride=1
    angularSpringConstant=16000
    bAngularSpring=true
    bOverrideAngularDamping=true
    bUseGravityOverride=true
    gravityScale=0.35
    simulationSpace=World
    boundBone {boneName=dyn_lowerLeg_r}
    componentAppliedLinearAccClamp=<5000 5000 5000>
    componentLinearAccScale=<0.8 0.8 0.8>
    componentLinearVelScale=<0.8 0.8 0.8>
    gravityOverride {z=-950}
    physicsBodyDefinitions [{
      boundBone {boneName=dyn_lowerLeg_r}
      boxExtents=<27.5 27.5 27.5>
      constraintSetup {angularLimitsMax=<0 0 17>, angularLimitsMin=<0 -1 -20>, angularTarget=<1 0 0>}
      localJointOffset=<0.85 0 0>
     }]
   }
   n6 = Anim Dynamics - Bone: dyn_lowerLeg_l(ComponentPose=@n5, Alpha=1, ExternalForce="0,0,0") {
    angularDampingOverride=1
    angularSpringConstant=17000
    bAngularSpring=true
    bOverrideAngularDamping=true
    bUseGravityOverride=true
    gravityScale=0.35
    simulationSpace=World
    boundBone {boneName=dyn_lowerLeg_l}
    componentAppliedLinearAccClamp=<3000 3000 3000>
    componentLinearAccScale=<0.65 0.65 0.65>
    componentLinearVelScale=<0.65 0.65 0.65>
    gravityOverride {z=-1280}
    physicsBodyDefinitions [{
      boundBone {boneName=dyn_lowerLeg_l}
      boxExtents=<25.5 30.5 30.5>
      constraintSetup {angularLimitsMax=<5 0 12.5>, angularLimitsMin=<0 0 -8>, angularTarget=<1 0 0>}
      localJointOffset=<0.77 0 0>
     }]
   }
   n7 = Component To Local(ComponentPose=@n6)
   n8 = Anim Dynamics - Bone: dyn_eye_r(ComponentPose=Local To Component(LocalPose=Use cached pose 'InputCachePose'()), Alpha=1, ExternalForce="0,0,0") {
    angularDampingOverride=1
    angularSpringConstant=17000
    bAngularSpring=true
    bOverrideAngularDamping=true
    bUseGravityOverride=true
    gravityScale=0.35
    simulationSpace=RootRelative
    worldSpaceMinimumScale=0
    boundBone {boneName=dyn_eye_r}
    componentAppliedLinearAccClamp=<3000 3000 3000>
    componentLinearAccScale=<0.5 0.5 0.5>
    componentLinearVelScale=<0.5 0.5 0.5>
    gravityOverride {z=-980}
    physicsBodyDefinitions [{
      boundBone {boneName=dyn_eye_r}
      boxExtents=<12.5 12.5 12.5>
      constraintSetup {angularLimitsMax=<0 37 37>, angularLimitsMin=<0 -37 -37>, angularTarget=<1 0 0>}
      localJointOffset=<4.5 0 0>
     }]
    simSpaceSettings {maxAngularAcceleration=800, maxAngularVelocity=80, simSpaceAngularAlpha=0.2}
   }
   n9 = Anim Dynamics - Bone: dyn_eye_l(ComponentPose=@n8, Alpha=1, ExternalForce="0,0,0") {
    angularDampingOverride=1
    angularSpringConstant=17000
    bAngularSpring=true
    bOverrideAngularDamping=true
    bUseGravityOverride=true
    gravityScale=0.35
    simulationSpace=RootRelative
    worldSpaceMinimumScale=0
    boundBone {boneName=dyn_eye_l}
    componentAppliedLinearAccClamp=<3000 3000 3000>
    componentLinearAccScale=<0.5 0.5 0.5>
    componentLinearVelScale=<0.5 0.5 0.5>
    gravityOverride {z=-980}
    physicsBodyDefinitions [{
      boundBone {boneName=dyn_eye_l}
      boxExtents=<12.5 12.5 12.5>
      constraintSetup {angularLimitsMax=<0 32 23>, angularLimitsMin=<0 -23 -32>, angularTarget=<1 0 0>}
      localJointOffset=<4.5 0 0>
     }]
    simSpaceSettings {maxAngularAcceleration=800, maxAngularVelocity=80, simSpaceAngularAlpha=0.2}
   }
   n10 = Anim Dynamics - Bone: dyn_upperLeg_r(ComponentPose=@n9, Alpha=1, ExternalForce="0,0,0") {
    angularDampingOverride=1
    angularSpringConstant=15000
    bAngularSpring=true
    bOverrideAngularDamping=true
    bUseGravityOverride=true
    gravityScale=0.35
    simulationSpace=World
    boundBone {boneName=dyn_upperLeg_r}
    componentAppliedLinearAccClamp=<2000 2000 2000>
    componentLinearAccScale=<0.5 0.5 0.5>
    componentLinearVelScale=<0.5 0.5 0.5>
    gravityOverride {z=-980}
    physicsBodyDefinitions [{
      boundBone {boneName=dyn_upperLeg_r}
      boxExtents=<34 34 34>
      constraintSetup {angularLimitsMax=<0 0.5 26>, angularLimitsMin=<0 -0.5 -19>, angularTarget=<1 0 0>}
      localJointOffset=<0.75 0 0>
     }]
   }
   n11 = Anim Dynamics - Bone: dyn_upperLeg_l(ComponentPose=@n10, Alpha=1, ExternalForce="0,0,0") {
    angularDampingOverride=1
    angularSpringConstant=17000
    bAngularSpring=true
    bOverrideAngularDamping=true
    bUseGravityOverride=true
    gravityScale=0.35
    simulationSpace=World
    boundBone {boneName=dyn_upperLeg_l}
    componentAppliedLinearAccClamp=<5000 5000 5000>
    componentLinearAccScale=<0.8 0.8 0.8>
    componentLinearVelScale=<0.8 0.8 0.8>
    gravityOverride {z=-950}
    physicsBodyDefinitions [{
      boundBone {boneName=dyn_upperLeg_l}
      boxExtents=<32 32 32>
      constraintSetup {angularLimitsMax=<0 2 23>, angularLimitsMin=<0 -1 -20>, angularTarget=<1 0 0>}
      localJointOffset=<0.82 0 0>
     }]
   }
   n12 = Anim Dynamics - Bone: dyn_lowerLeg_r(ComponentPose=@n11, Alpha=1, ExternalForce="0,0,0") {
    angularDampingOverride=1
    angularSpringConstant=16000
    bAngularSpring=true
    bOverrideAngularDamping=true
    bUseGravityOverride=true
    gravityScale=0.35
    simulationSpace=World
    boundBone {boneName=dyn_lowerLeg_r}
    componentAppliedLinearAccClamp=<5000 5000 5000>
    componentLinearAccScale=<0.8 0.8 0.8>
    componentLinearVelScale=<0.8 0.8 0.8>
    gravityOverride {z=-950}
    physicsBodyDefinitions [{
      boundBone {boneName=dyn_lowerLeg_r}
      boxExtents=<25 27 27>
      constraintSetup {angularLimitsMax=<0 0 17>, angularLimitsMin=<0 -1 -20>, angularTarget=<1 0 0>}
      localJointOffset=<1 0 0>
     }]
   }
   n13 = Anim Dynamics - Bone: dyn_lowerLeg_l(ComponentPose=@n12, Alpha=1, ExternalForce="0,0,0") {
    angularDampingOverride=1
    angularSpringConstant=17000
    bAngularSpring=true
    bOverrideAngularDamping=true
    bUseGravityOverride=true
    gravityScale=0.35
    simulationSpace=World
    boundBone {boneName=dyn_lowerLeg_l}
    componentAppliedLinearAccClamp=<3000 3000 3000>
    componentLinearAccScale=<0.65 0.65 0.65>
    componentLinearVelScale=<0.65 0.65 0.65>
    gravityOverride {z=-1280}
    physicsBodyDefinitions [{
      boundBone {boneName=dyn_lowerLeg_l}
      boxExtents=<25 30 30>
      constraintSetup {angularLimitsMax=<5 0 12.5>, angularLimitsMin=<0 0 -8>, angularTarget=<1 0 0>}
      localJointOffset=<0.8 0 0>
     }]
   }
   n14 = Component To Local(ComponentPose=@n13)
   n15 = Get bIsOnPawn() -> (bIsOnPawn:bool)
   n16 = Anim Dynamics - Bone: dyn_eye_r(ComponentPose=Local To Component(LocalPose=Use cached pose 'InputCachePose'()), Alpha=1, ExternalForce="0,0,0") {
    angularDampingOverride=1
    angularSpringConstant=18000
    bAngularSpring=true
    bOverrideAngularDamping=true
    bUseGravityOverride=true
    gravityScale=0.35
    simulationSpace=BoneRelative
    boundBone {boneName=dyn_eye_r}
    componentAppliedLinearAccClamp=<6000 6000 6000>
    componentLinearAccScale=<0.8 0.8 0.8>
    componentLinearVelScale=<0.8 0.8 8>
    gravityOverride {z=-980}
    physicsBodyDefinitions [{
      boundBone {boneName=dyn_eye_r}
      boxExtents=<20 27 27>
      constraintSetup {angularLimitsMax=<0 25 25>, angularLimitsMin=<0 -25 -25>, angularTarget=<1 0 0>}
      localJointOffset=<3.8 0 0>
     }]
    simSpaceSettings {maxAngularAcceleration=100, maxAngularVelocity=10, simSpaceAngularAlpha=0.5}
   }
   n17 = Anim Dynamics - Bone: dyn_eye_l(ComponentPose=@n16, Alpha=1, ExternalForce="0,0,0") {
    angularDampingOverride=1
    angularSpringConstant=18000
    bAngularSpring=true
    bOverrideAngularDamping=true
    bUseGravityOverride=true
    gravityScale=0.35
    simulationSpace=BoneRelative
    boundBone {boneName=dyn_eye_l}
    componentAppliedLinearAccClamp=<6000 6000 6000>
    componentLinearAccScale=<0.8 0.8 0.8>
    componentLinearVelScale=<0.8 0.8 0.8>
    gravityOverride {z=-980}
    physicsBodyDefinitions [{
      boundBone {boneName=dyn_eye_l}
      boxExtents=<20 27 27>
      constraintSetup {angularLimitsMax=<0 28 18>, angularLimitsMin=<0 -18 -28>, angularTarget=<1 0 0>}
      localJointOffset=<3.8 0 0>
     }]
    simSpaceSettings {maxAngularAcceleration=100, maxAngularVelocity=10, simSpaceAngularAlpha=0.5}
   }
   n18 = Anim Dynamics - Bone: dyn_upperLeg_r(ComponentPose=@n17, Alpha=1, ExternalForce="0,0,0") {
    angularDampingOverride=1
    angularSpringConstant=17000
    bAngularSpring=true
    bOverrideAngularDamping=true
    bUseGravityOverride=true
    gravityScale=0.35
    simulationSpace=BoneRelative
    boundBone {boneName=dyn_upperLeg_r}
    componentAppliedLinearAccClamp=<4000 4000 4000>
    componentLinearAccScale=<0.65 0.65 0.65>
    componentLinearVelScale=<0.65 0.65 0.65>
    gravityOverride {z=-1200}
    physicsBodyDefinitions [{
      boundBone {boneName=dyn_upperLeg_r}
      boxExtents=<25 30 30>
      constraintSetup {angularLimitsMax=<0 10 35>, angularLimitsMin=<0 -10 -29>, angularTarget=<1 0 0>}
      localJointOffset=<3.5 0 0>
     }]
   }
   n19 = Anim Dynamics - Bone: dyn_upperLeg_l(ComponentPose=@n18, Alpha=1, ExternalForce="0,0,0") {
    angularDampingOverride=1
    angularSpringConstant=17000
    bAngularSpring=true
    bOverrideAngularDamping=true
    bUseGravityOverride=true
    gravityScale=0.35
    simulationSpace=BoneRelative
    boundBone {boneName=dyn_upperLeg_l}
    componentAppliedLinearAccClamp=<5000 5000 5000>
    componentLinearAccScale=<0.8 0.8 0.8>
    componentLinearVelScale=<0.8 0.8 0.8>
    gravityOverride {z=-950}
    physicsBodyDefinitions [{
      boundBone {boneName=dyn_upperLeg_l}
      boxExtents=<24 28 28>
      constraintSetup {angularLimitsMax=<0 19 42>, angularLimitsMin=<-35 -15 -39>, angularTarget=<1 0 0>}
      localJointOffset=<3.8 0 0>
     }]
   }
   n20 = Anim Dynamics - Bone: dyn_lowerLeg_r(ComponentPose=@n19, Alpha=1, ExternalForce="0,0,0") {
    angularDampingOverride=1
    angularSpringConstant=16000
    bAngularSpring=true
    bOverrideAngularDamping=true
    bUseGravityOverride=true
    gravityScale=0.35
    simulationSpace=BoneRelative
    boundBone {boneName=dyn_lowerLeg_r}
    componentAppliedLinearAccClamp=<5000 5000 5000>
    componentLinearAccScale=<0.8 0.8 0.8>
    componentLinearVelScale=<0.8 0.8 0.8>
    gravityOverride {z=-950}
    physicsBodyDefinitions [{
      boundBone {boneName=dyn_lowerLeg_r}
      boxExtents=<25 28 28>
      constraintSetup {angularLimitsMax=<0 17 39>, angularLimitsMin=<0 -19 -42>, angularTarget=<1 0 0>}
      localJointOffset=<3.75 0 0>
     }]
   }
   n21 = Anim Dynamics - Bone: dyn_lowerLeg_l(ComponentPose=@n20, Alpha=1, ExternalForce="0,0,0") {
    angularDampingOverride=1
    angularSpringConstant=17000
    bAngularSpring=true
    bOverrideAngularDamping=true
    bUseGravityOverride=true
    gravityScale=0.35
    simulationSpace=BoneRelative
    boundBone {boneName=dyn_lowerLeg_l}
    componentAppliedLinearAccClamp=<3000 3000 3000>
    componentLinearAccScale=<0.65 0.65 0.65>
    componentLinearVelScale=<0.65 0.65 0.65>
    gravityOverride {z=-1280}
    physicsBodyDefinitions [{
      boundBone {boneName=dyn_lowerLeg_l}
      boxExtents=<25 31 31>
      constraintSetup {angularLimitsMax=<10 10 25>, angularLimitsMin=<0 -10 -18>, angularTarget=<1 0 0>}
      localJointOffset=<3.5 0 0>
     }]
   }
   Output Pose(Result=Blend Poses by bool(
     BlendPose_0=Blend Poses by bool(BlendPose_0=@n7, BlendPose_1=@n14, bActiveValue=@n15)
     BlendPose_1=Component To Local(ComponentPose=@n21)
     bActiveValue=Get bIsInFrontEnd()
    )) {layerGroup=None}
   AnimGraphNode_SubInput_0 = Input Pose()
   Save cached pose 'InputCachePose'(Pose=@AnimGraphNode_SubInput_0) {cacheName=InputCachePose}
  ]}
 EventGraph {nodes [
   TryGetPawnOwner()
   Event BlueprintUpdateAnimation()
  ]}
]
variables []
